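# Sound cue for the engine loop foley of the SequinPie emote, played through the emote foley sound class and attenuation.
# A mixer layers looping idle and forward-motion waves, each through a continuous modulator driven by the EmoteMoveSpeed parameter.
# The modulators vary volume and pitch with speed, shifting the balance from idle toward forward motion.
# Concurrency is capped by an override, and the cue keeps playing when silent.
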SoundCue /BRCosmetics/Sounds/Emotes/SequinPie/EmoteFoley_SequinPie_Engine_Cue
root=Mixer_0
nodes [
 {id=Mixer_0, children [ModulatorContinuous_0, ModulatorContinuous_1], properties {inputVolume [1.5,1]}}
 {
  id=ModulatorContinuous_0
  children [WavePlayer_2]
  properties {
   pitchModulationParams {maxOutput=1.05, minOutput=1, parameterName=EmoteMoveSpeed}
   volumeModulationParams {minOutput=1.5, parameterName=EmoteMoveSpeed}
  }
 }
 {
  id=ModulatorContinuous_1
  children [WavePlayer_1]
  properties {pitchModulationParams {minOutput=0.98, parameterName=EmoteMoveSpeed}, volumeModulationParams {parameterName=EmoteMoveSpeed}}
 }
 {id=WavePlayer_1, properties {bLooping=true, soundWaveAssetPtr=&/IP_Emote_SequinPie_Fwd_LP}}
 {id=WavePlayer_2, properties {bLooping=true, soundWaveAssetPtr=&/IP_Emote_SequinPie_Idle_LP}}
]
properties {
 attenuationSettings=&Fort_Emotes_Athena_Att
 audioPropertiesSheet=/BRCosmetics/Sounds/Settings/PropertySheets/Emotes/APS_Cosmetic_Emote_Foley
 bOverrideConcurrency=true
 maxDistance=1600
 soundClassObject=/Game/Sounds/FortSoundClassesAndModes/Fort_Foley_Emote
 virtualizationMode=PlayWhenSilent
 volumeMultiplier=1
 concurrencyOverrides {maxCount=3}
 soundSubmixSends [{minSendLevel=0, soundSubmix=&Fortnite_AudioAnalysis_Music}]
}
